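# Anim Blueprint for the DashSurge backpack cosmetic, adding secondary motion to its sushi, salmon, onion, leaf, tongue and eye bones.
# One live Anim Dynamics chain feeds an additive pack-bone transform, then weighted constraints on eye and tongue master bones; other sushi chains are unconnected.
# Onion springs run only when neither skydiving nor DBNO, as IsSkyDivingOrDBNO stores a NOR despite its name.
# The EventGraph also computes gravity override and spine-yaw-driven pack rotation and translation.
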AnimBlueprint /BRCosmetics/Accessories/FORT_Backpacks/Backpack_M_MED_DashSurge/Meshes/M_MED_DashSurge_Pack_AnimBP
target_skeleton=&/M_MED_DashSurge_Pack_Skeleton
graphs [
 AnimGraph {nodes [
   n1 = Anim Dynamics - Bone: dyn_sushi_01_l!(ComponentPose, Alpha=1, GravityOverride:Vector="0,0,0", ExternalForce="0,0,70") -> (Pose) {
    angularDampingOverride=0.9999
    angularSpringConstant=1000
    bAngularSpring=true
    bLinearSpring=true
    bOverrideAngularDamping=true
    bOverrideLinearDamping=true
    gravityScale=0.1
    linearDampingOverride=0.99
    linearSpringConstant=80
    alphaBoolBlend {blendOption=Linear}
    alphaScaleBiasClamp {bClampResult=true, clampMin=0.5}
    boundBone {boneName=dyn_sushi_01_l}
    componentAppliedLinearAccClamp=<300 300 100>
    componentLinearAccScale=<0.3 0.3 0.3>
    componentLinearVelScale=<0.2 0.2 0.2>
    physicsBodyDefinitions [{
      boundBone {boneName=dyn_sushi_01_l}
      boxExtents=<15 15 15>
      constraintSetup {angularLimitsMax=<5 10 5>, angularLimitsMin=<-0.5 0 0>, angularTarget=<1 0 0>, linearAxesMax=<3 1 1>, linearAxesMin=<-0.5 0 -0.5>}
      localJointOffset=<1 0 0>
     }]
   }
   n2 = Anim Dynamics - Bone: dyn_sushi_01_r!(ComponentPose=@n1, Alpha=1, GravityOverride:Vector="0,0,0", ExternalForce="0,0,30") -> (Pose) {
    angularDampingOverride=0.999
    angularSpringConstant=500
    bAngularSpring=true
    bLinearSpring=true
    bOverrideAngularDamping=true
    bOverrideLinearDamping=true
    gravityScale=0.1
    linearDampingOverride=0.99
    linearSpringConstant=80
    numSolverIterationsPreUpdate=3
    alphaBoolBlend {blendOption=Linear}
    alphaScaleBiasClamp {bClampResult=true, clampMin=0.2}
    boundBone {boneName=dyn_sushi_01_r}
    componentAppliedLinearAccClamp=<450 450 100>
    componentLinearAccScale=<0.2 0.2 0.2>
    componentLinearVelScale=<0.1 0.1 0.1>
    physicsBodyDefinitions [{
      boundBone {boneName=dyn_sushi_01_r}
      boxExtents=<20 20 20>
      constraintSetup {angularLimitsMax=<7 7 7>, angularLimitsMin=<-0.5 0 0>, angularTarget=<1 0 0>, linearAxesMax=<3 1 1>, linearAxesMin=<-0.5 0 -0.5>}
      localJointOffset=<0.5 0 0>
     }]
   }
   Anim Dynamics - Bone: dyn_sushi_c!(ComponentPose=@n2, Alpha=1, GravityOverride:Vector="0,0,0", ExternalForce="0,0,70") -> (Pose) {
    angularDampingOverride=0.999
    angularSpringConstant=200
    bAngularSpring=true
    bLinearSpring=true
    bOverrideAngularDamping=true
    bOverrideLinearDamping=true
    gravityScale=0.1
    linearDampingOverride=0.99
    linearSpringConstant=50
    alphaBoolBlend {blendOption=Linear}
    alphaScaleBiasClamp {bClampResult=true}
    boundBone {boneName=dyn_sushi_c}
    componentAppliedLinearAccClamp=<175 175 70>
    componentLinearAccScale=<0.4 0.3 0.4>
    componentLinearVelScale=<0.3 0.4 0.3>
    physicsBodyDefinitions [{
      boundBone {boneName=dyn_sushi_c}
      boxExtents=<25 25 25>
      constraintSetup {angularLimitsMax=<5 5 5>, angularLimitsMin=<-0.5 0 0>, angularTarget=<1 0 0>, linearAxesMax=<3 1 1>, linearAxesMin=<-0.5 0 -0.5>}
      localJointOffset=<1.5 0 0>
     }]
   }
   n3 = Anim Dynamics - Bone: dyn_sushi_01_l(Alpha=1, GravityOverride="0,0,0", ExternalForce="0,0,70") {
    angularDampingOverride=0.9999
    angularSpringConstant=1000
    bAngularSpring=true
    bLinearSpring=true
    bOverrideAngularDamping=true
    bOverrideLinearDamping=true
    gravityScale=0
    linearDampingOverride=0.99
    linearSpringConstant=80
    alphaBoolBlend {blendOption=Linear}
    alphaScaleBiasClamp {bClampResult=true, clampMin=0.5}
    boundBone {boneName=dyn_sushi_01_l}
    componentAppliedLinearAccClamp=<500 500 200>
    componentLinearAccScale=<0.3 0.3 0.3>
    componentLinearVelScale=<0.2 0.2 0.2>
    physicsBodyDefinitions [{
      boundBone {boneName=dyn_sushi_01_l}
      boxExtents=<15 15 15>
      constraintSetup {angularLimitsMax=<5 10 5>, angularLimitsMin=<-5 -10 -5>, angularTarget=<1 0 0>, linearAxesMax=<3 1 1>, linearAxesMin=<-2 -1 -1>}
      localJointOffset=<1 0 0>
     }]
   }
   n4 = Anim Dynamics - Bone: dyn_sushi_01_r(ComponentPose=@n3, Alpha=1, GravityOverride="0,0,0", ExternalForce="0,0,30") {
    angularDampingOverride=0.99
    angularSpringConstant=500
    bAngularSpring=true
    bLinearSpring=true
    bOverrideAngularDamping=true
    bOverrideLinearDamping=true
    gravityScale=0
    linearDampingOverride=0.998
    linearSpringConstant=80
    numSolverIterationsPreUpdate=3
    alphaBoolBlend {blendOption=Linear}
    alphaScaleBiasClamp {bClampResult=true, clampMin=0.2}
    boundBone {boneName=dyn_sushi_01_r}
    componentAppliedLinearAccClamp=<700 700 150>
    componentLinearAccScale=<0.2 0.2 0.2>
    componentLinearVelScale=<0.1 0.1 0.1>
    physicsBodyDefinitions [{
      boundBone {boneName=dyn_sushi_01_r}
      boxExtents=<20 20 20>
      constraintSetup {angularLimitsMax=<7 7 7>, angularLimitsMin=<-7 -7 -7>, angularTarget=<1 0 0>, linearAxesMax=<3 1 1>, linearAxesMin=<-2 -1 -1>}
      localJointOffset=<0.5 0 0>
     }]
   }
   Anim Dynamics - Bone: dyn_sushi_c(ComponentPose=@n4, Alpha=1, GravityOverride="0,0,0", ExternalForce="0,0,70") {
    angularDampingOverride=0.999
    angularSpringConstant=200
    bAngularSpring=true
    bLinearSpring=true
    bOverrideAngularDamping=true
    bOverrideLinearDamping=true
    gravityScale=0
    linearDampingOverride=0.98
    linearSpringConstant=50
    alphaBoolBlend {blendOption=Linear}
    alphaScaleBiasClamp {bClampResult=true}
    boundBone {boneName=dyn_sushi_c}
    componentAppliedLinearAccClamp=<250 250 100>
    componentLinearAccScale=<0.4 0.3 0.4>
    componentLinearVelScale=<0.3 0.4 0.3>
    physicsBodyDefinitions [{
      boundBone {boneName=dyn_sushi_c}
      boxExtents=<25 25 25>
      constraintSetup {angularLimitsMax=<5 5 5>, angularLimitsMin=<-5 -5 -5>, angularTarget=<1 0 0>, linearAxesMax=<3 1 1>, linearAxesMin=<-2 -1 -1>}
      localJointOffset=<1.5 0 0>
     }]
   }
   n5 = Anim Dynamics - Bone: dyn_sushi_01_l(Alpha=1, GravityOverride="0,0,0", ExternalForce="0,0,70") {
    angularDampingOverride=0.9999
    angularSpringConstant=1000
    bAngularSpring=true
    bLinearSpring=true
    bOverrideAngularDamping=true
    bOverrideLinearDamping=true
    gravityScale=0
    linearDampingOverride=0.999
    linearSpringConstant=80
    alphaBoolBlend {blendOption=Linear}
    alphaScaleBiasClamp {bClampResult=true, clampMin=0.5}
    boundBone {boneName=dyn_sushi_01_l}
    componentAppliedLinearAccClamp=<300 300 100>
    componentLinearAccScale=<0.3 0.3 0.3>
    componentLinearVelScale=<0.2 0.2 0.2>
    physicsBodyDefinitions [{
      boundBone {boneName=dyn_sushi_01_l}
      boxExtents=<15 15 15>
      constraintSetup {angularLimitsMax=<5 10 5>, angularLimitsMin=<-5 -10 -5>, angularTarget=<1 0 0>, linearAxesMax=<3 1 1>, linearAxesMin=<-2 -1 -1>}
      localJointOffset=<1 0 0>
     }]
   }
   n6 = Anim Dynamics - Bone: dyn_sushi_01_r(ComponentPose=@n5, Alpha=1, GravityOverride="0,0,0", ExternalForce="0,0,30") {
    angularDampingOverride=0.998
    angularSpringConstant=500
    bAngularSpring=true
    bLinearSpring=true
    bOverrideAngularDamping=true
    bOverrideLinearDamping=true
    gravityScale=0
    linearDampingOverride=0.9988
    linearSpringConstant=80
    numSolverIterationsPreUpdate=3
    alphaBoolBlend {blendOption=Linear}
    alphaScaleBiasClamp {bClampResult=true, clampMin=0.2}
    boundBone {boneName=dyn_sushi_01_r}
    componentAppliedLinearAccClamp=<450 450 100>
    componentLinearAccScale=<0.2 0.2 0.2>
    componentLinearVelScale=<0.1 0.1 0.1>
    physicsBodyDefinitions [{
      boundBone {boneName=dyn_sushi_01_r}
      boxExtents=<20 20 20>
      constraintSetup {angularLimitsMax=<7 7 7>, angularLimitsMin=<-7 -7 -7>, angularTarget=<1 0 0>, linearAxesMax=<3 1 1>, linearAxesMin=<-2 -1 -1>}
      localJointOffset=<0.5 0 0>
     }]
   }
   Anim Dynamics - Bone: dyn_sushi_c(ComponentPose=@n6, Alpha=1, GravityOverride="0,0,0", ExternalForce="0,0,70") {
    angularDampingOverride=0.999
    angularSpringConstant=200
    bAngularSpring=true
    bLinearSpring=true
    bOverrideAngularDamping=true
    bOverrideLinearDamping=true
    gravityScale=0
    linearDampingOverride=0.9888
    linearSpringConstant=50
    alphaBoolBlend {blendOption=Linear}
    alphaScaleBiasClamp {bClampResult=true}
    boundBone {boneName=dyn_sushi_c}
    componentAppliedLinearAccClamp=<175 175 70>
    componentLinearAccScale=<0.4 0.3 0.4>
    componentLinearVelScale=<0.3 0.4 0.3>
    physicsBodyDefinitions [{
      boundBone {boneName=dyn_sushi_c}
      boxExtents=<25 25 25>
      constraintSetup {angularLimitsMax=<5 5 5>, angularLimitsMin=<-5 -5 -5>, angularTarget=<1 0 0>, linearAxesMax=<3 1 1>, linearAxesMin=<-2 -1 -1>}
      localJointOffset=<1.5 0 0>
     }]
   }
   n7 = Anim Dynamics - Bone: dyn_sushi_01_l(Alpha=1, GravityOverride="0,0,0", ExternalForce="0,0,70") {
    angularDampingOverride=0.9999
    angularSpringConstant=1000
    bAngularSpring=true
    bLinearSpring=true
    bOverrideAngularDamping=true
    bOverrideLinearDamping=true
    gravityScale=0.5
    linearDampingOverride=0.998
    linearSpringConstant=80
    alphaBoolBlend {blendOption=Linear}
    alphaScaleBiasClamp {bClampResult=true, clampMin=0.5}
    boundBone {boneName=dyn_sushi_01_l}
    componentAppliedLinearAccClamp=<300 300 100>
    componentLinearAccScale=<0.3 0.3 0.3>
    componentLinearVelScale=<0.2 0.2 0.2>
    physicsBodyDefinitions [{
      boundBone {boneName=dyn_sushi_01_l}
      boxExtents=<15 15 15>
      constraintSetup {angularLimitsMax=<5 10 5>, angularLimitsMin=<-5 -10 -5>, angularTarget=<1 0 0>, linearAxesMax=<3 1 1>, linearAxesMin=<-2 -1 -1>}
      localJointOffset=<1 0 0>
     }]
   }
   n8 = Anim Dynamics - Bone: dyn_sushi_01_r(ComponentPose=@n7, Alpha=1, GravityOverride="0,0,0", ExternalForce="0,0,30") {
    angularDampingOverride=0.999
    angularSpringConstant=500
    bAngularSpring=true
    bLinearSpring=true
    bOverrideAngularDamping=true
    bOverrideLinearDamping=true
    gravityScale=0.5
    linearDampingOverride=0.998
    linearSpringConstant=80
    numSolverIterationsPreUpdate=3
    alphaBoolBlend {blendOption=Linear}
    alphaScaleBiasClamp {bClampResult=true, clampMin=0.2}
    boundBone {boneName=dyn_sushi_01_r}
    componentAppliedLinearAccClamp=<450 450 100>
    componentLinearAccScale=<0.2 0.2 0.2>
    componentLinearVelScale=<0.1 0.1 0.1>
    physicsBodyDefinitions [{
      boundBone {boneName=dyn_sushi_01_r}
      boxExtents=<20 20 20>
      constraintSetup {angularLimitsMax=<7 7 7>, angularLimitsMin=<-7 -7 -7>, angularTarget=<1 0 0>, linearAxesMax=<3 1 1>, linearAxesMin=<-2 -1 -1>}
      localJointOffset=<0.5 0 0>
     }]
   }
   Anim Dynamics - Bone: dyn_sushi_c(ComponentPose=@n8, Alpha=1, GravityOverride="0,0,0", ExternalForce="0,0,70") {
    angularDampingOverride=0.999
    angularSpringConstant=200
    bAngularSpring=true
    bLinearSpring=true
    bOverrideAngularDamping=true
    bOverrideLinearDamping=true
    gravityScale=0.5
    linearDampingOverride=0.999
    linearSpringConstant=50
    alphaBoolBlend {blendOption=Linear}
    alphaScaleBiasClamp {bClampResult=true}
    boundBone {boneName=dyn_sushi_c}
    componentAppliedLinearAccClamp=<175 175 70>
    componentLinearAccScale=<0.4 0.3 0.4>
    componentLinearVelScale=<0.3 0.4 0.3>
    physicsBodyDefinitions [{
      boundBone {boneName=dyn_sushi_c}
      boxExtents=<25 25 25>
      constraintSetup {angularLimitsMax=<5 5 5>, angularLimitsMin=<-5 -5 -5>, angularTarget=<1 0 0>, linearAxesMax=<3 1 1>, linearAxesMin=<-2 -1 -1>}
      localJointOffset=<1.5 0 0>
     }]
   }
   n9 = Anim Dynamics - Bone: dyn_sushi_01_l!(ComponentPose=Local To Component(LocalPose=Input Pose() {name=InPose_0}), Alpha=1, ExternalForce="0,0,70") -> (Pose) {
    angularDampingOverride=0.9999
    angularSpringConstant=2000
    bAngularSpring=true
    bLinearSpring=true
    bOverrideAngularDamping=true
    bOverrideLinearDamping=true
    gravityScale=0.5
    linearDampingOverride=0.999
    linearSpringConstant=180
    alphaBoolBlend {blendOption=Linear}
    alphaScaleBiasClamp {bClampResult=true, clampMin=0.5}
    boundBone {boneName=dyn_sushi_01_l}
    componentAppliedLinearAccClamp=<300 300 100>
    componentLinearAccScale=<0.3 0.3 0.3>
    componentLinearVelScale=<0.2 0.2 0.2>
    gravityOverride {z=-500}
    physicsBodyDefinitions [{
      boundBone {boneName=dyn_sushi_01_l}
      boxExtents=<15 15 15>
      constraintSetup {angularLimitsMax=<5 8 5>, angularLimitsMin=<-0.5 0 0>, angularTarget=<1 0 0>, linearAxesMax=<2 0.7 1>, linearAxesMin=<-0.5 0 -0.5>}
      localJointOffset=<0.7 0 0>
     }]
   }
   n10 = Anim Dynamics - Bone: dyn_sushi_01_r!(ComponentPose=@n9, Alpha=1, ExternalForce="0,0,30") -> (Pose) {
    angularDampingOverride=0.999
    angularSpringConstant=2000
    bAngularSpring=true
    bLinearSpring=true
    bOverrideAngularDamping=true
    bOverrideLinearDamping=true
    gravityScale=0.5
    linearDampingOverride=0.999
    linearSpringConstant=200
    numSolverIterationsPreUpdate=3
    alphaBoolBlend {blendOption=Linear}
    alphaScaleBiasClamp {bClampResult=true, clampMin=0.2}
    boundBone {boneName=dyn_sushi_01_r}
    componentAppliedLinearAccClamp=<450 450 100>
    componentLinearAccScale=<0.2 0.2 0.2>
    componentLinearVelScale=<0.1 0.1 0.1>
    gravityOverride {z=-500}
    physicsBodyDefinitions [{
      boundBone {boneName=dyn_sushi_01_r}
      boxExtents=<20 20 20>
      constraintSetup {angularLimitsMax=<5 5 5>, angularLimitsMin=<-0.5 0 0>, angularTarget=<1 0 0>, linearAxesMax=<2 1 1>, linearAxesMin=<-0.5 0 -0.5>}
      localJointOffset=<0.9 0 0>
     }]
   }
   n11 = Anim Dynamics - Bone: dyn_sushi_c!(ComponentPose=@n10, Alpha=1, ExternalForce="0,0,70") -> (Pose) {
    angularDampingOverride=0.999
    angularSpringConstant=2000
    bAngularSpring=true
    bLinearSpring=true
    bOverrideAngularDamping=true
    bOverrideLinearDamping=true
    gravityScale=0.5
    linearDampingOverride=0.9999
    linearSpringConstant=80
    alphaBoolBlend {blendOption=Linear}
    alphaScaleBiasClamp {bClampResult=true}
    boundBone {boneName=dyn_sushi_c}
    componentAppliedLinearAccClamp=<175 175 70>
    componentLinearAccScale=<0.4 0.3 0.4>
    componentLinearVelScale=<0.3 0.4 0.3>
    gravityOverride {z=-500}
    physicsBodyDefinitions [{
      boundBone {boneName=dyn_sushi_c}
      boxExtents=<25 25 25>
      constraintSetup {angularLimitsMax=<3 3 3>, angularLimitsMin=<-0.5 0 0>, angularTarget=<1 0 0>, linearAxesMax=<2 0.7 1>, linearAxesMin=<-0.5 0 -0.5>}
      localJointOffset=<0.8 0 0>
     }]
   }
   n12 = Anim Dynamics - Bone: dyn_salmon_r!(ComponentPose=@n11, Alpha=1, GravityOverride:Vector="0,0,0", ExternalForce="0,0,0") -> (Pose) {
    angularDampingOverride=0.9
    angularSpringConstant=50
    bAngularSpring=true
    bOverrideAngularDamping=true
    bUseGravityOverride=true
    alphaBoolBlend {blendOption=Linear}
    alphaScaleBiasClamp {bClampResult=true}
    boundBone {boneName=dyn_salmon_r}
    componentAppliedLinearAccClamp=<500 500 1200>
    componentLinearAccScale=<0.5 0.5 0.8>
    componentLinearVelScale=<0.5 0.5 0.5>
    physicsBodyDefinitions [{
      boundBone {boneName=dyn_salmon_r}
      boxExtents=<12 12 12>
      constraintSetup {angularLimitsMax=<0 -10 10>, angularLimitsMin=<0 -40 -10>, angularTarget=<0 0 0>}
      localJointOffset=<1.5 0 0>
     }]
   }
   n13 = Anim Dynamics - Bone: dyn_salmon_l!(ComponentPose=@n12, Alpha=1, GravityOverride:Vector="0,0,0", ExternalForce="0,0,0") -> (Pose) {
    angularDampingOverride=0.99
    bOverrideAngularDamping=true
    bUseGravityOverride=true
    alphaBoolBlend {blendOption=Linear}
    alphaScaleBiasClamp {bClampResult=true}
    boundBone {boneName=dyn_salmon_l}
    componentAppliedLinearAccClamp=<500 500 1100>
    componentLinearAccScale=<0.5 0.5 0.5>
    componentLinearVelScale=<0.6 0.6 0.6>
    physicsBodyDefinitions [{
      boundBone {boneName=dyn_salmon_l}
      boxExtents=<12 10 10>
      constraintSetup {angularLimitsMax=<0 -20 5>, angularLimitsMin=<0 -60 -5>, angularTarget=<0 0 0>}
      localJointOffset=<2 0 0>
     }]
   }
   n14 = Get IsSkyDivingOrDBNO() -> (IsSkyDivingOrDBNO:bool)
   n15 = Anim Dynamics - Bone: dyn_onion_01!(ComponentPose=@n13, bAlphaBoolEnabled:bool=@n14, ExternalForce="0,0,0") -> (Pose) {
    alphaInputType=Bool
    bLinearSpring=true
    bOverrideLinearDamping=true
    bUseGravityOverride=true
    linearDampingOverride=0.9
    linearSpringConstant=30
    alphaBoolBlend {blendOption=Linear}
    alphaScaleBiasClamp {bClampResult=true}
    boundBone {boneName=dyn_onion_01}
    componentAppliedLinearAccClamp=<0 0 1000>
    componentLinearAccScale {z=0.5}
    componentLinearVelScale {z=0.6}
    gravityOverride {z=-980}
    physicsBodyDefinitions [{
      boundBone {boneName=dyn_onion_01}
      boxExtents=<5 5 5>
      constraintSetup {angularLimitsMax=<0 0 0>, angularLimitsMin=<0 0 0>, angularTarget=<0 0 0>, linearAxesMax=<5 0 0>}
      localJointOffset=<1 0 0>
     }]
   }
   n16 = Anim Dynamics - Bone: dyn_onion_02!(ComponentPose=@n15, bAlphaBoolEnabled:bool=@n14, ExternalForce="0,0,0") -> (Pose) {
    alphaInputType=Bool
    bLinearSpring=true
    bOverrideLinearDamping=true
    bUseGravityOverride=true
    linearDampingOverride=0.99
    linearSpringConstant=5
    alphaBoolBlend {blendOption=Linear}
    alphaScaleBiasClamp {bClampResult=true, clampMax=0.6}
    boundBone {boneName=dyn_onion_02}
    componentAppliedLinearAccClamp=<0 0 1000>
    componentLinearAccScale {z=0.5}
    componentLinearVelScale {z=0.6}
    gravityOverride {z=-960}
    physicsBodyDefinitions [{
      boundBone {boneName=dyn_onion_02}
      boxExtents=<5 10 10>
      constraintSetup {angularLimitsMax=<0 0 0>, angularLimitsMin=<0 0 0>, angularTarget=<0 0 0>, linearAxesMax=<5 0 0>}
      localJointOffset=<0.5 0 0>
     }]
   }
   n17 = Anim Dynamics - Bone: dyn_onion_03!(ComponentPose=@n16, bAlphaBoolEnabled:bool=@n14, ExternalForce="0,0,0") -> (Pose) {
    alphaInputType=Bool
    bLinearSpring=true
    bOverrideLinearDamping=true
    bUseGravityOverride=true
    linearDampingOverride=0.98
    linearSpringConstant=50
    alphaBoolBlend {blendOption=Linear}
    alphaScaleBiasClamp {bClampResult=true, clampMin=0.3}
    boundBone {boneName=dyn_onion_03}
    componentAppliedLinearAccClamp=<0 0 1000>
    componentLinearAccScale {z=0.5}
    componentLinearVelScale {z=0.6}
    gravityOverride {z=-900}
    physicsBodyDefinitions [{
      boundBone {boneName=dyn_onion_03}
      boxExtents=<5 5 5>
      constraintSetup {angularLimitsMax=<0 0 0>, angularLimitsMin=<0 0 0>, angularTarget=<0 0 0>, linearAxesMax=<5 0 0>}
      localJointOffset=<1 0 0>
     }]
   }
   n18 = Get IsSkyDivingOrDBNO()
   n19 = Anim Dynamics - Bone: dyn_onion_04!(ComponentPose=@n17, bAlphaBoolEnabled:bool=@n18, ExternalForce="0,0,0") -> (Pose) {
    alphaInputType=Bool
    bLinearSpring=true
    bOverrideLinearDamping=true
    bUseGravityOverride=true
    linearDampingOverride=0.8
    linearSpringConstant=10
    alphaBoolBlend {blendOption=Linear}
    alphaScaleBiasClamp {bClampResult=true, clampMin=0.5}
    boundBone {boneName=dyn_onion_04}
    componentAppliedLinearAccClamp=<0 0 1000>
    componentLinearAccScale {z=0.5}
    componentLinearVelScale {z=0.6}
    gravityOverride {z=-980}
    physicsBodyDefinitions [{
      boundBone {boneName=dyn_onion_04}
      boxExtents=<8 8 8>
      constraintSetup {angularLimitsMax=<0 0 0>, angularLimitsMin=<0 0 0>, angularTarget=<0 0 0>, linearAxesMax=<5 0 0>}
      localJointOffset=<0.6 0 0>
     }]
   }
   n20 = Anim Dynamics - Bone: dyn_onion_05!(ComponentPose=@n19, bAlphaBoolEnabled:bool=@n18, ExternalForce="0,0,0") -> (Pose) {
    alphaInputType=Bool
    bLinearSpring=true
    bOverrideLinearDamping=true
    bUseGravityOverride=true
    linearDampingOverride=0.9
    linearSpringConstant=20
    alphaBoolBlend {blendOption=Linear}
    alphaScaleBiasClamp {bClampResult=true, clampMax=0.8}
    boundBone {boneName=dyn_onion_05}
    componentAppliedLinearAccClamp=<0 0 1000>
    componentLinearAccScale {z=0.5}
    componentLinearVelScale {z=0.6}
    gravityOverride {z=-920}
    physicsBodyDefinitions [{
      boundBone {boneName=dyn_onion_05}
      boxExtents=<10 5 5>
      constraintSetup {angularLimitsMax=<0 0 0>, angularLimitsMin=<0 0 0>, angularTarget=<0 0 0>, linearAxesMax=<5 0 0>}
      localJointOffset=<0.7 0 0>
     }]
   }
   n21 = Anim Dynamics - Bone: dyn_onion_06!(ComponentPose=@n20, bAlphaBoolEnabled:bool=@n18, ExternalForce="0,0,0") -> (Pose) {
    alphaInputType=Bool
    bLinearSpring=true
    bOverrideLinearDamping=true
    bUseGravityOverride=true
    linearDampingOverride=0.99
    linearSpringConstant=50
    alphaBoolBlend {blendOption=Linear}
    alphaScaleBiasClamp {bClampResult=true}
    boundBone {boneName=dyn_onion_06}
    componentAppliedLinearAccClamp=<0 0 1000>
    componentLinearAccScale {z=0.5}
    componentLinearVelScale {z=0.6}
    gravityOverride {z=-940}
    physicsBodyDefinitions [{
      boundBone {boneName=dyn_onion_06}
      boxExtents=<10 10 5>
      constraintSetup {angularLimitsMax=<0 0 0>, angularLimitsMin=<0 0 0>, angularTarget=<0 0 0>, linearAxesMax=<5 0 0>}
      localJointOffset=<1 0 0>
     }]
   }
   n22 = Anim Dynamics - Bone: dyn_sushi_tongue_r!(ComponentPose=@n21, Alpha=1, GravityOverride:Vector="0,0,0", ExternalForce="0,0,0") -> (Pose) {
    angularDampingOverride=0.98
    angularSpringConstant=100
    bAngularSpring=true
    bOverrideAngularDamping=true
    bUseGravityOverride=true
    alphaBoolBlend {blendOption=Linear}
    alphaScaleBiasClamp {bClampResult=true}
    boundBone {boneName=dyn_sushi_tongue_r}
    componentAppliedLinearAccClamp=<1000 1000 1000>
    componentLinearAccScale=<0.5 0.5 0.5>
    componentLinearVelScale=<0.7 0.7 0.7>
    physicsBodyDefinitions [{
      boundBone {boneName=dyn_sushi_tongue_r}
      boxExtents=<10 5 5>
      constraintSetup {angularLimitsMax=<0 90 20>, angularLimitsMin=<0 -5 -20>, angularTarget=<1 0 0>}
      localJointOffset=<1 0 0>
     }]
   }
   n23 = Anim Dynamics - Bone: dyn_leaf_01!(ComponentPose=@n22, Alpha=1, GravityOverride:Vector="0,0,0", ExternalForce="0,0,0") -> (Pose) {
    angularDampingOverride=0.999
    angularSpringConstant=100
    bAngularSpring=true
    bOverrideAngularDamping=true
    bUseGravityOverride=true
    alphaBoolBlend {blendOption=Linear}
    alphaScaleBiasClamp {bClampResult=true}
    boundBone {boneName=dyn_leaf_01}
    componentAppliedLinearAccClamp=<1500 1500 1000>
    componentLinearAccScale=<0.7 0.7 0.7>
    componentLinearVelScale=<0.7 0.7 0.7>
    physicsBodyDefinitions [{
      boundBone {boneName=dyn_leaf_01}
      boxExtents=<10 10 10>
      constraintSetup {angularLimitsMax=<0 0 120>, angularLimitsMin=<0 0 0>, angularTarget=<0 0 0>}
      localJointOffset=<2 0 0>
     }]
   }
   n24 = Anim Dynamics - Bone: dyn_leaf_02!(ComponentPose=@n23, Alpha=1, GravityOverride:Vector="0,0,0", ExternalForce="0,0,0") -> (Pose) {
    angularDampingOverride=0.998
    angularSpringConstant=80
    bAngularSpring=true
    bOverrideAngularDamping=true
    bUseGravityOverride=true
    alphaBoolBlend {blendOption=Linear}
    alphaScaleBiasClamp {bClampResult=true}
    boundBone {boneName=dyn_leaf_02}
    componentAppliedLinearAccClamp=<1500 1500 1000>
    componentLinearAccScale=<0.5 0.5 0.5>
    componentLinearVelScale=<0.7 0.7 0.7>
    physicsBodyDefinitions [{
      boundBone {boneName=dyn_leaf_02}
      boxExtents=<8 8 8>
      constraintSetup {angularLimitsMax=<0 5 70>, angularLimitsMin=<0 -5 -20>, angularTarget=<1 0 0>}
      localJointOffset=<1.5 0 0>
     }]
   }
   n25 = Anim Dynamics - Bone: dyn_leaf_03!(ComponentPose=@n24, Alpha=1, GravityOverride:Vector="0,0,0", ExternalForce="0,0,0") -> (Pose) {
    angularDampingOverride=0.999
    angularSpringConstant=60
    bAngularSpring=true
    bOverrideAngularDamping=true
    bUseGravityOverride=true
    alphaBoolBlend {blendOption=Linear}
    alphaScaleBiasClamp {bClampResult=true}
    boundBone {boneName=dyn_leaf_03}
    componentAppliedLinearAccClamp=<1500 1500 1000>
    componentLinearAccScale=<0.4 0.4 0.4>
    componentLinearVelScale=<0.5 0.5 0.5>
    physicsBodyDefinitions [{
      boundBone {boneName=dyn_leaf_03}
      boxExtents=<6 6 6>
      constraintSetup {angularLimitsMax=<0 5 60>, angularLimitsMin=<0 -5 -30>, angularTarget=<1 0 0>}
      localJointOffset=<1 0 0>
     }]
   }
   n26 = Anim Dynamics - Bone: dyn_eye_01_r(ComponentPose=@n25, Alpha=1, ExternalForce="0,0,0") {
    angularDampingOverride=1
    angularSpringConstant=2000
    bAngularSpring=true
    bGravityOverrideInSimSpace=true
    bOverrideAngularDamping=true
    gravityScale=0
    numSolverIterationsPreUpdate=3
    alphaBoolBlend {blendOption=Linear}
    alphaScaleBiasClamp {bClampResult=true}
    boundBone {boneName=dyn_eye_01_r}
    componentAppliedLinearAccClamp=<2000 2000 2500>
    componentLinearAccScale=<0.2 0.4 0.7>
    componentLinearVelScale=<0.5 0.4 0.8>
    physicsBodyDefinitions [{
      boundBone {boneName=dyn_eye_01_r}
      boxExtents=<10 10 10>
      constraintSetup {angularLimitsMax=<35 35 35>, angularLimitsMin=<-35 -35 -35>, angularTarget=<1 0 0>}
      localJointOffset=<1 1.5 1>
     }]
   }
   n27 = Anim Dynamics - Bone: dyn_eye_02_r(ComponentPose=@n26, Alpha=1, ExternalForce="0,0,0") {
    angularDampingOverride=1
    angularSpringConstant=2000
    bAngularSpring=true
    bGravityOverrideInSimSpace=true
    bOverrideAngularDamping=true
    gravityScale=0
    alphaBoolBlend {blendOption=Linear}
    alphaScaleBiasClamp {bClampResult=true}
    boundBone {boneName=dyn_eye_02_r}
    componentAppliedLinearAccClamp=<1500 1500 1500>
    componentLinearAccScale=<0.4 0.4 0.4>
    componentLinearVelScale=<0.5 0.5 0.5>
    physicsBodyDefinitions [{
      boundBone {boneName=dyn_eye_02_r}
      boxExtents=<12 12 8>
      constraintSetup {angularLimitsMax=<35 35 35>, angularLimitsMin=<-35 -35 -35>, angularTarget=<1 0 0>}
      localJointOffset=<1.5 1 1.5>
     }]
   }
   n28 = Anim Dynamics - Bone: dyn_eye_01_l(ComponentPose=@n27, Alpha=1, ExternalForce="0,0,0") {
    angularDampingOverride=1
    angularSpringConstant=3000
    bAngularSpring=true
    bGravityOverrideInSimSpace=true
    bOverrideAngularDamping=true
    gravityScale=0
    numSolverIterationsPreUpdate=3
    alphaBoolBlend {blendOption=Linear}
    alphaScaleBiasClamp {bClampResult=true}
    boundBone {boneName=dyn_eye_01_l}
    componentAppliedLinearAccClamp=<2000 2000 2500>
    componentLinearAccScale=<0.2 0.4 0.7>
    componentLinearVelScale=<0.5 0.4 0.8>
    physicsBodyDefinitions [{
      boundBone {boneName=dyn_eye_01_l}
      boxExtents=<10 10 10>
      constraintSetup {angularLimitsMax=<35 35 35>, angularLimitsMin=<-35 -35 -35>, angularTarget=<1 0 0>}
      localJointOffset=<1 1 1>
     }]
   }
   n29 = Anim Dynamics - Bone: dyn_eye_02_l(ComponentPose=@n28, Alpha=1, ExternalForce="0,0,0") {
    angularDampingOverride=1
    angularSpringConstant=2000
    bAngularSpring=true
    bGravityOverrideInSimSpace=true
    bOverrideAngularDamping=true
    gravityScale=0
    alphaBoolBlend {blendOption=Linear}
    alphaScaleBiasClamp {bClampResult=true}
    boundBone {boneName=dyn_eye_02_l}
    componentAppliedLinearAccClamp=<1500 1500 1500>
    componentLinearAccScale=<0.4 0.4 0.4>
    componentLinearVelScale=<0.5 0.5 0.5>
    physicsBodyDefinitions [{
      boundBone {boneName=dyn_eye_02_l}
      boxExtents=<12 12 8>
      constraintSetup {angularLimitsMax=<35 35 35>, angularLimitsMin=<-35 -35 -35>, angularTarget=<1 0 0>}
      localJointOffset=<1 2 1>
     }]
   }
   n30 = Anim Dynamics - Bone: dyn_eye_02_c(ComponentPose=@n29, Alpha=1, ExternalForce="0,0,0") {
    angularDampingOverride=1
    angularSpringConstant=3000
    bAngularSpring=true
    bGravityOverrideInSimSpace=true
    bOverrideAngularDamping=true
    gravityScale=0
    alphaBoolBlend {blendOption=Linear}
    alphaScaleBiasClamp {bClampResult=true}
    boundBone {boneName=dyn_eye_02_c}
    componentAppliedLinearAccClamp=<2000 2000 2500>
    componentLinearAccScale=<0.2 0.4 0.7>
    componentLinearVelScale=<0.5 0.4 0.8>
    physicsBodyDefinitions [{
      boundBone {boneName=dyn_eye_02_c}
      boxExtents=<10 10 10>
      constraintSetup {angularLimitsMax=<25 25 25>, angularLimitsMin=<-25 -25 -25>, angularTarget=<1 0 0>}
      localJointOffset=<1 1.5 1>
     }]
   }
   n31 = Anim Dynamics - Bone: dyn_eye_01_c(ComponentPose=@n30, Alpha=1, ExternalForce="0,0,0") {
    angularDampingOverride=1
    angularSpringConstant=2000
    bAngularSpring=true
    bGravityOverrideInSimSpace=true
    bOverrideAngularDamping=true
    gravityScale=0
    alphaBoolBlend {blendOption=Linear}
    alphaScaleBiasClamp {bClampResult=true}
    boundBone {boneName=dyn_eye_01_c}
    componentAppliedLinearAccClamp=<1500 1500 2000>
    componentLinearAccScale=<0.4 0.4 0.4>
    componentLinearVelScale=<0.5 0.5 0.5>
    physicsBodyDefinitions [{
      boundBone {boneName=dyn_eye_01_c}
      boxExtents=<12 12 8>
      constraintSetup {angularLimitsMax=<25 25 25>, angularLimitsMin=<-25 -25 -25>, angularTarget=<1 0 0>}
      localJointOffset=<1.5 0.5 1.5>
     }]
   }
   n32 = "Transform (Modify) Bone - Bone: pack"(ComponentPose=@n31, Translation="0,0,0", Rotation="0,0,0") {
    rotationMode=BMM_Additive
    rotationSpace=BCS_BoneSpace
    translationMode=BMM_Additive
    translationSpace=BCS_BoneSpace
    alphaBoolBlend {blendOption=Linear}
    boneToModify {boneName=pack}
   }
   n33 = "Constraint - dyn_sushi_tongue_MSTR (PP)"!(ComponentPose=@n32, Alpha=1, ConstraintWeights_0=0.295, ConstraintWeights_1:float=0.705) -> (Pose) {
    alphaBoolBlend {blendOption=Linear}
    boneToModify {boneName=dyn_sushi_tongue_MSTR}
    constraintSetup [{targetBone {boneName=dyn_sushi_01_r}}, {targetBone {boneName=dyn_sushi_base}}]
    constraintWeights [1,1]
   }
   n34 = "Constraint - dyn_eye_01_c_MSTR (PPP)"!(ComponentPose=@n33, Alpha=1, ConstraintWeights_0=0.017, ConstraintWeights_1:float=0.076, ConstraintWeights_2:float=0.907) -> (Pose) {
    alphaBoolBlend {blendOption=Linear}
    boneToModify {boneName=dyn_eye_01_c_MSTR}
    constraintSetup [{targetBone {boneName=dyn_sushi_01_l}}, {targetBone {boneName=dyn_sushi_01_r}}, {targetBone {boneName=dyn_sushi_c}}]
    constraintWeights [1,1,1]
   }
   n35 = "Constraint - dyn_eye_02_c_MSTR (PPP)"!(ComponentPose=@n34, Alpha=1, ConstraintWeights_0=0.054, ConstraintWeights_1:float=0.005, ConstraintWeights_2:float=0.941) -> (Pose) {
    alphaBoolBlend {blendOption=Linear}
    boneToModify {boneName=dyn_eye_02_c_MSTR}
    constraintSetup [{targetBone {boneName=dyn_sushi_01_l}}, {targetBone {boneName=dyn_sushi_01_r}}, {targetBone {boneName=dyn_sushi_c}}]
    constraintWeights [1,1,1]
   }
   n36 = "Constraint - dyn_eye_01_r_MSTR (PP)"!(ComponentPose=@n35, Alpha=1, ConstraintWeights_0=0.648, ConstraintWeights_1:float=0.352) -> (Pose) {
    alphaBoolBlend {blendOption=Linear}
    boneToModify {boneName=dyn_eye_01_r_MSTR}
    constraintSetup [{targetBone {boneName=dyn_sushi_01_r}}, {targetBone {boneName=dyn_sushi_base}}]
    constraintWeights [1,1]
   }
   n37 = "Constraint - dyn_eye_02_r_MSTR (PP)"!(ComponentPose=@n36, Alpha=1, ConstraintWeights_0=0.584, ConstraintWeights_1:float=0.416) -> (Pose) {
    alphaBoolBlend {blendOption=Linear}
    boneToModify {boneName=dyn_eye_02_r_MSTR}
    constraintSetup [{targetBone {boneName=dyn_sushi_01_r}}, {targetBone {boneName=dyn_sushi_base}}]
    constraintWeights [1,1]
   }
   n38 = "Constraint - dyn_eye_01_l_MSTR (PP)"!(ComponentPose=@n37, Alpha=1, ConstraintWeights_0=0.666, ConstraintWeights_1:float=0.334) -> (Pose) {
    alphaBoolBlend {blendOption=Linear}
    boneToModify {boneName=dyn_eye_01_l_MSTR}
    constraintSetup [{targetBone {boneName=dyn_sushi_01_l}}, {targetBone {boneName=dyn_sushi_base}}]
    constraintWeights [1,1]
   }
   n39 = "Constraint - dyn_eye_02_l_MSTR (PP)"!(ComponentPose=@n38, Alpha=1, ConstraintWeights_0=0.676, ConstraintWeights_1:float=0.324) -> (Pose) {
    alphaBoolBlend {blendOption=Linear}
    boneToModify {boneName=dyn_eye_02_l_MSTR}
    constraintSetup [{targetBone {boneName=dyn_sushi_01_l}}, {targetBone {boneName=dyn_sushi_base}}]
    constraintWeights [1,1]
   }
   Output Pose(Result=Component To Local(ComponentPose=@n39)) {layerGroup=None}
  ]}
 EventGraph {nodes [
   Event BlueprintUpdateAnimation()
   > n1 = Gravity Override State Function DEPRECATED(CharacterPartAnimInstance=Self-Reference(), GameTimeInSeconds=GetGameTimeInSeconds(), GravityOverrideParams=Get Gravity Override Params())
   > Set Gravity Override(Gravity Override=@n1)
   > Set IsSkyDivingOrDBNO(IsSkyDivingOrDBNO:bool=NOR Boolean!(A:bool=Get bIsSkyDiving(), B:bool=Get bIsDBNO()) -> (ReturnValue:bool)) -> (Output_Get:bool)
   n2 = GetOwnerSkeletalMeshComponent()
   n3 = BreakRotator(InRot=BreakTransform(InTransform=GetDeltaTransformFromRefPose(self=@n2, BoneName=spine_05, BaseName=spine_04)).Rotation)
   > Set BackPack Rotation(BackPack Rotation:Rotator=MakeRotator(Roll=MapRangeClamped(Value=@n3.Yaw, InRangeA=0, InRangeB=-30, OutRangeB=40))) -> (Output_Get:Rotator)
   > Set BackPack Translation(BackPack Translation:Vector=MakeVector(Y=MapRangeClamped(Value=@n3.Yaw, InRangeA=0, InRangeB=-30, OutRangeB=-2))) -> (Output_Get:Vector)
   TryGetPawnOwner()
  ]}
]
variables [
 BackPack Rotation {type=Rotator}
 BackPack Translation {type=Vector}
 Gravity Override {type=Vector}
 Gravity Override Params {type=GravityOverrideParamsStruct}
 IsSkyDivingOrDBNO {type=boolean}
]
